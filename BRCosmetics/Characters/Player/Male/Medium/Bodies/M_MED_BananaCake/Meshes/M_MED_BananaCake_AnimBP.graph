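# Animation Blueprint for the M_MED_BananaCake player body cosmetic, adding secondary motion to shorts, buckle, bags and map.
# It copies the parent mesh pose through a Control Rig and per-leg RigidBody shorts simulations into a cached pose.
# AnimDynamics, constraints, calf-driven Flap nodes and additive bone offsets then refine that pose, some gated by bool flags.
# The EventGraph derives flags and floats from movement, weapon type, joint rotations and LOD.
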AnimBlueprint /BRCosmetics/Characters/Player/Male/Medium/Bodies/M_MED_BananaCake/Meshes/M_MED_BananaCake_AnimBP
target_skeleton=&/M_MED_BananaCake_Skeleton
graphs [
 AnimGraph {nodes [
   "Constraint - dyn_shorts_bk_r (PPPPPPP)"!(
    ComponentPose
    Alpha=1
    ConstraintWeights_0=0.1396
    ConstraintWeights_1:float=0.391
    ConstraintWeights_2:float=0.0058
    ConstraintWeights_3:float=0.0944
    ConstraintWeights_4:float=0.0577
    ConstraintWeights_5:float=0.3036
    ConstraintWeights_6:float=0.0078
   ) -> (Pose) {
    alphaBoolBlend {blendInTime=10, blendOption=Linear, blendOutTime=10}
    alphaScaleBias {scale=0.5}
    alphaScaleBiasClamp {bInterpResult=true}
    boneToModify {boneName=dyn_shorts_bk_r}
    constraintSetup [
     {targetBone {boneName=thigh_r}}
     {targetBone {boneName=calf_r}}
     {targetBone {boneName=calf_twist_01_r}}
     {targetBone {boneName=calf_twist_02_r}}
     {targetBone {boneName=deform_knee_out_r}}
     {targetBone {boneName=deform_knee_in_r}}
     {targetBone {boneName=thigh_twist_01_r}}
    ]
    constraintWeights [1,1,1,1,1,1,1]
   }
   n1 = "Constraint - thigh_twist_01_l (P)"!(ComponentPose, bAlphaBoolEnabled:bool=True, ConstraintWeights_0=1) -> (Pose) {
    alphaInputType=Bool
    alphaBoolBlend {blendOption=Linear}
    boneToModify {boneName=thigh_twist_01_l}
    constraintSetup [{targetBone {boneName=thigh_l}}]
    constraintWeights [1]
   }
   "Constraint - thigh_twist_01_r (P)"!(ComponentPose=@n1, bAlphaBoolEnabled:bool=True, ConstraintWeights_0=1) -> (Pose) {
    alphaInputType=Bool
    alphaBoolBlend {blendOption=Linear}
    boneToModify {boneName=thigh_twist_01_r}
    constraintSetup [{targetBone {boneName=thigh_r}}]
    constraintWeights [1]
   }
   "Transform (Modify) Bone - Bone: dyn_shorts_bk_r"!(ComponentPose, bAlphaBoolEnabled:bool=True, Translation="-4.337,0.156,17.125", Rotation="14.688,-0.002,-3.973", Scale="1.4,1,1") -> (Pose) {
    alphaInputType=Bool
    rotationMode=BMM_Additive
    rotationSpace=BCS_BoneSpace
    scaleMode=BMM_Additive
    scaleSpace=BCS_BoneSpace
    translationMode=BMM_Additive
    translationSpace=BCS_BoneSpace
    alphaBoolBlend {blendInTime=0.5, blendOption=Linear, blendOutTime=0.7}
    boneToModify {boneName=dyn_shorts_bk_r}
    rotation=<rot 14.688 -0.002 -3.973>
    scale {x=1.4}
    translation=<-4.337 0.156 17.125>
   }
   n2 = Anim Dynamics - Bone: dyn_map(ComponentPose=Local To Component(LocalPose=Use cached pose 'PreJacketSim'()), Alpha=0.05, ExternalForce="0,0,0") {
    angularDampingOverride=1
    bOverrideAngularDamping=true
    bUseGravityOverride=true
    gravityScale=0.35
    simulationSpace=RootRelative
    alphaBoolBlend {blendOption=Linear}
    boundBone {boneName=dyn_map}
    componentAppliedLinearAccClamp=<5000 5000 5000>
    gravityOverride {z=-800}
    physicsBodyDefinitions [{
      boundBone {boneName=dyn_map}
      boxExtents=<5 10 10>
      constraintSetup {angularLimitsMax=<10 45 20>, angularLimitsMin=<-10 -45 -20>, angularTarget=<1 0 0>}
      localJointOffset=<4 0 0>
     }]
   }
   n3 = Anim Dynamics - Bone: dyn_mapFront(ComponentPose=@n2, Alpha=0.1, ExternalForce="0,0,0") {
    angularDampingOverride=0.9999
    bOverrideAngularDamping=true
    bUseGravityOverride=true
    gravityScale=0.35
    simulationSpace=RootRelative
    alphaBoolBlend {blendOption=Linear}
    boundBone {boneName=dyn_mapFront}
    componentAppliedLinearAccClamp=<5000 5000 5000>
    gravityOverride {z=-800}
    physicsBodyDefinitions [{
      boundBone {boneName=dyn_mapFront}
      boxExtents=<5 10 10>
      constraintSetup {angularLimitsMax=<10 45 20>, angularLimitsMin=<-10 -45 -20>, angularTarget=<1 0 0>}
      localJointOffset=<4 0 0>
     }]
   }
   n4 = Anim Dynamics - Bone: dyn_mapBack(ComponentPose=@n3, Alpha=0.12, ExternalForce="0,0,0") {
    angularDampingOverride=0.9999
    bOverrideAngularDamping=true
    bUseGravityOverride=true
    gravityScale=0.35
    simulationSpace=RootRelative
    alphaBoolBlend {blendOption=Linear}
    boundBone {boneName=dyn_mapBack}
    componentAppliedLinearAccClamp=<5000 5000 5000>
    gravityOverride {z=-800}
    physicsBodyDefinitions [{
      boundBone {boneName=dyn_mapBack}
      boxExtents=<5 10 10>
      constraintSetup {angularLimitsMax=<10 45 20>, angularLimitsMin=<-10 -45 -20>, angularTarget=<1 0 0>}
      localJointOffset=<4 0 0>
     }]
   }
   n5 = "Constraint - dyn_buckle_MSTR (PPPPPPPP)"!(
    ComponentPose=@n4
    Alpha=1
    ConstraintWeights_0=0.4649
    ConstraintWeights_1:float=0.0216
    ConstraintWeights_2:float=0.0392
    ConstraintWeights_3:float=0.0255
    ConstraintWeights_4:float=0.1476
    ConstraintWeights_5:float=0.0934
    ConstraintWeights_6:float=0.1079
    ConstraintWeights_7:float=0.0998
   ) -> (Pose) {
    alphaBoolBlend {blendOption=Linear}
    boneToModify {boneName=dyn_buckle_MSTR}
    constraintSetup [
     {targetBone {boneName=pelvis}}
     {targetBone {boneName=spine_01}}
     {targetBone {boneName=spine_02}}
     {targetBone {boneName=spine_03}}
     {targetBone {boneName=thigh_twist_01_l}}
     {targetBone {boneName=deform_groin_l}}
     {targetBone {boneName=thigh_twist_01_r}}
     {targetBone {boneName=deform_groin_r}}
    ]
    constraintWeights [1,1,1,1,1,1,1,1]
   }
   n6 = "Constraint - dyn_buckle_01 (RR)"!(ComponentPose=@n5, bAlphaBoolEnabled:bool=True, ConstraintWeights_0=0.5, ConstraintWeights_1:float=0.5) -> (Pose) {
    alphaInputType=Bool
    alphaBoolBlend {blendInTime=0.1, blendOption=Linear, blendOutTime=0.1}
    boneToModify {boneName=dyn_buckle_01}
    constraintSetup [{transformType=Rotation, targetBone {boneName=thigh_l}}, {transformType=Rotation, targetBone {boneName=thigh_r}}]
    constraintWeights [1,1]
   }
   n7 = Get bCheckCurrentLod() -> (bCheckCurrentLod:bool)
   n8 = "Constraint - dyn_buckle_MSTR (P)"!(ComponentPose=@n6, bAlphaBoolEnabled:bool=@n7, ConstraintWeights_0=1) -> (Pose) {
    alphaInputType=Bool
    alphaBoolBlend {blendOption=Linear}
    boneToModify {boneName=dyn_buckle_MSTR}
    constraintSetup [{offsetOption=None, targetBone {boneName=pelvis}}]
    constraintWeights [1]
   }
   n9 = "Transform (Modify) Bone - Bone: dyn_buckle_MSTR"!(ComponentPose=@n8, bAlphaBoolEnabled:bool=@n7, Translation="-6.455,21.889,2.243", Rotation="1.097,84.778,-75.671", Scale="1,1,1") -> (Pose) {
    alphaInputType=Bool
    rotationMode=BMM_Additive
    rotationSpace=BCS_BoneSpace
    scaleMode=BMM_Additive
    scaleSpace=BCS_BoneSpace
    translationMode=BMM_Additive
    translationSpace=BCS_BoneSpace
    alphaBoolBlend {blendInTime=0.05, blendOption=Linear, blendOutTime=0.1}
    boneToModify {boneName=dyn_buckle_MSTR}
    rotation=<rot 1.097 84.778 -75.671>
    translation=<-6.455 21.889 2.243>
   }
   n10 = "Transform (Modify) Bone - Bone: dyn_buckle_MSTR"(
    ComponentPose=@n9
    bAlphaBoolEnabled=Get bForBucketConstCond() -> (bForBucketConstCond:bool)
    Translation="-1.000,-3.993,6.708"
    Rotation="2.484,-20.230,-3.125"
    Scale="1,1,1"
   ) {
    alphaInputType=Bool
    rotationMode=BMM_Additive
    rotationSpace=BCS_BoneSpace
    scaleMode=BMM_Additive
    scaleSpace=BCS_BoneSpace
    translationMode=BMM_Additive
    translationSpace=BCS_BoneSpace
    alphaBoolBlend {blendInTime=0.1, blendOption=Linear, blendOutTime=0.1}
    boneToModify {boneName=dyn_buckle_MSTR}
    rotation=<rot 2.484 -20.23 -3.125>
    translation=<-1 -3.993 6.708>
   }
   n11 = "Transform (Modify) Bone - Bone: dyn_buckle_MSTR"(ComponentPose=@n10, bAlphaBoolEnabled=True, Translation="2.468,1.488,0.109", Rotation="-21.401,25.536,2.967", Scale="1,1,1") {
    alphaInputType=Bool
    rotationMode=BMM_Additive
    rotationSpace=BCS_BoneSpace
    scaleMode=BMM_Additive
    scaleSpace=BCS_BoneSpace
    translationMode=BMM_Additive
    translationSpace=BCS_BoneSpace
    alphaBoolBlend {blendInTime=0.3, blendOption=Linear, blendOutTime=1}
    boneToModify {boneName=dyn_buckle_MSTR}
    rotation=<rot -21.401 25.536 2.967>
    translation=<2.468 1.488 0.109>
   }
   n12 = Flap Node - Bone: dyn_shorts_bk_l!(ComponentPose:Component Space Pose Link Structure=@n11, Alpha:float=1) -> (Pose:Component Space Pose Link Structure) {
    is=title FortAnimGraphNode_Flap
    targetComponent=RotationY
    alphaBoolBlend {blendOption=Linear}
    alphaScaleBiasClamp {bInterpResult=true, interpSpeedDecreasing=0.1}
    sourceDrives [{
      bUseRange=true
      drivingCurve=null
      multiplier=0.5
      rangeMax=-0.53
      rangeMin=-2.094
      remappedMax=0
      remappedMin=-1.919
      sourceComponent=RotationZ
      useQuaternion=true
      sourceBone {boneName=calf_l}
     }]
    targetBone {boneName=dyn_shorts_bk_l}
   }
   n13 = Flap Node - Bone: dyn_shorts_bk_l(ComponentPose=@n12, Alpha=1) {
    targetComponent=ScaleZ
    alphaBoolBlend {blendOption=Linear}
    alphaScaleBiasClamp {bInterpResult=true, interpSpeedDecreasing=0.1}
    sourceDrives [{
      bUseRange=true
      drivingCurve=null
      multiplier=0.5
      rangeMax=-0.53
      rangeMin=-2.094
      remappedMax=0
      remappedMin=-0.8
      sourceComponent=RotationZ
      useQuaternion=true
      sourceBone {boneName=calf_l}
     }]
    targetBone {boneName=dyn_shorts_bk_l}
   }
   n14 = Flap Node - Bone: dyn_shorts_bk_r!(ComponentPose:Component Space Pose Link Structure=@n13, Alpha:float=1) -> (Pose:Component Space Pose Link Structure) {
    is=title FortAnimGraphNode_Flap
    targetComponent=RotationY
    alphaBoolBlend {blendOption=Linear}
    alphaScaleBiasClamp {bInterpResult=true, interpSpeedDecreasing=0.1}
    sourceDrives [{
      bUseRange=true
      drivingCurve=null
      multiplier=0.5
      rangeMax=-0.53
      rangeMin=-2.094
      remappedMax=0
      remappedMin=-1.99
      sourceComponent=RotationZ
      useQuaternion=true
      sourceBone {boneName=calf_r}
     }]
    targetBone {boneName=dyn_shorts_bk_r}
   }
   n15 = Flap Node - Bone: dyn_shorts_bk_r(ComponentPose=@n14, Alpha=1) {
    targetComponent=ScaleZ
    alphaBoolBlend {blendOption=Linear}
    alphaScaleBiasClamp {bInterpResult=true, interpSpeedDecreasing=0.1}
    sourceDrives [{
      bUseRange=true
      drivingCurve=null
      multiplier=0.5
      rangeMax=-0.53
      rangeMin=-2.094
      remappedMax=0
      remappedMin=-0.9
      sourceComponent=RotationZ
      useQuaternion=true
      sourceBone {boneName=calf_r}
     }]
    targetBone {boneName=dyn_shorts_bk_r}
   }
   n16 = "Constraint - dyn_shorts_bk_r (R)"(ComponentPose=@n15, Alpha=1, ConstraintWeights_0=1) {
    alphaBoolBlend {blendInTime=0.3, blendOption=Linear, blendOutTime=0.3}
    alphaScaleBiasClamp {bInterpResult=true}
    boneToModify {boneName=dyn_shorts_bk_r}
    constraintSetup [{transformType=Rotation, targetBone {boneName=calf_r}}]
    constraintWeights [1]
   }
   n17 = "Constraint - dyn_shorts_bk_l (R)"(ComponentPose=@n16, Alpha=1, ConstraintWeights_0=1) {
    alphaBoolBlend {blendInTime=0.3, blendOption=Linear, blendOutTime=0.3}
    alphaScaleBiasClamp {bInterpResult=true}
    boneToModify {boneName=dyn_shorts_bk_l}
    constraintSetup [{transformType=Rotation, targetBone {boneName=calf_l}}]
    constraintWeights [1]
   }
   n18 = "Constraint - dyn_shorts_in_r (R)"(ComponentPose=@n17, Alpha=0.5, ConstraintWeights_0=1) {
    alphaBoolBlend {blendInTime=0.3, blendOption=Linear, blendOutTime=0.3}
    alphaScaleBiasClamp {bInterpResult=true}
    boneToModify {boneName=dyn_shorts_in_r}
    constraintSetup [{transformType=Rotation, targetBone {boneName=calf_r}}]
    constraintWeights [1]
   }
   n19 = "Constraint - dyn_shorts_in_l (R)"(ComponentPose=@n18, Alpha=0.5, ConstraintWeights_0=1) {
    alphaBoolBlend {blendInTime=0.3, blendOption=Linear, blendOutTime=0.3}
    alphaScaleBiasClamp {bInterpResult=true}
    boneToModify {boneName=dyn_shorts_in_l}
    constraintSetup [{transformType=Rotation, targetBone {boneName=calf_l}}]
    constraintWeights [1]
   }
   n20 = "Constraint - dyn_shorts_out_r (R)"(ComponentPose=@n19, Alpha=0.5, ConstraintWeights_0=1) {
    alphaBoolBlend {blendInTime=0.3, blendOption=Linear, blendOutTime=0.3}
    alphaScaleBiasClamp {bInterpResult=true}
    boneToModify {boneName=dyn_shorts_out_r}
    constraintSetup [{transformType=Rotation, targetBone {boneName=calf_r}}]
    constraintWeights [1]
   }
   n21 = "Constraint - dyn_shorts_out_l (R)"(ComponentPose=@n20, Alpha=0.5, ConstraintWeights_0=1) {
    alphaBoolBlend {blendInTime=0.3, blendOption=Linear, blendOutTime=0.3}
    alphaScaleBiasClamp {bInterpResult=true}
    boneToModify {boneName=dyn_shorts_out_l}
    constraintSetup [{transformType=Rotation, targetBone {boneName=calf_l}}]
    constraintWeights [1]
   }
   n22 = "Transform (Modify) Bone - Bone: dyn_shorts_bk_r"(ComponentPose=@n21, bAlphaBoolEnabled=True, Translation="-11.460,-4.090,-8.857", Rotation="-12.453,0.000,0.000", Scale="1,1,1") {
    alphaInputType=Bool
    rotationMode=BMM_Additive
    rotationSpace=BCS_BoneSpace
    scaleMode=BMM_Additive
    scaleSpace=BCS_BoneSpace
    translationMode=BMM_Additive
    translationSpace=BCS_BoneSpace
    alphaBoolBlend {blendInTime=0.2, blendOption=Linear, blendOutTime=0.4}
    boneToModify {boneName=dyn_shorts_bk_r}
    rotation=<rot -12.453 0 0>
    translation=<-11.46 -4.09 -8.857>
   }
   n23 = "Transform (Modify) Bone - Bone: dyn_shorts_in_r"!(ComponentPose=@n22, bAlphaBoolEnabled:bool=True, Translation="-0.527,0.000,-3.788", Rotation="0.000,-0.000,17.848", Scale="1,1,1") -> (Pose) {
    alphaInputType=Bool
    rotationMode=BMM_Additive
    rotationSpace=BCS_BoneSpace
    scaleMode=BMM_Additive
    scaleSpace=BCS_BoneSpace
    translationMode=BMM_Additive
    translationSpace=BCS_BoneSpace
    alphaBoolBlend {blendInTime=0.2, blendOption=Linear, blendOutTime=0.4}
    boneToModify {boneName=dyn_shorts_in_r}
    rotation=<rot 0 0 17.848>
    translation {x=-0.527, z=-3.788}
   }
   n24 = "Transform (Modify) Bone - Bone: dyn_shorts_out_r"!(ComponentPose=@n23, bAlphaBoolEnabled:bool=True, Translation="-3.966,-0.117,-7.351", Rotation="-6.140,-8.403,-8.048", Scale="1,1,1") -> (Pose) {
    alphaInputType=Bool
    rotationMode=BMM_Additive
    rotationSpace=BCS_BoneSpace
    scaleMode=BMM_Additive
    scaleSpace=BCS_BoneSpace
    translationMode=BMM_Additive
    translationSpace=BCS_BoneSpace
    alphaBoolBlend {blendInTime=0.2, blendOption=Linear, blendOutTime=0.4}
    boneToModify {boneName=dyn_shorts_out_r}
    rotation=<rot -6.14 -8.403 -8.048>
    translation=<-3.966 -0.117 -7.351>
   }
   n25 = "Transform (Modify) Bone - Bone: dyn_shorts_fr_r"!(ComponentPose=@n24, bAlphaBoolEnabled:bool=True, Translation="0,0,0", Rotation="-2.170,11.018,10.178", Scale="1,1,1") -> (Pose) {
    alphaInputType=Bool
    rotationMode=BMM_Additive
    rotationSpace=BCS_BoneSpace
    scaleMode=BMM_Additive
    scaleSpace=BCS_BoneSpace
    translationMode=BMM_Additive
    translationSpace=BCS_BoneSpace
    alphaBoolBlend {blendInTime=0.2, blendOption=Linear, blendOutTime=0.4}
    boneToModify {boneName=dyn_shorts_fr_r}
    rotation=<rot -2.17 11.018 10.178>
   }
   n26 = "Transform (Modify) Bone - Bone: dyn_shorts_bk_l"!(ComponentPose=@n25, bAlphaBoolEnabled:bool=True, Translation="12.414,-0.000,-2.823", Rotation="-46.771,-0.000,0.000", Scale="1,1,1") -> (Pose) {
    alphaInputType=Bool
    rotationMode=BMM_Additive
    rotationSpace=BCS_BoneSpace
    scaleMode=BMM_Additive
    scaleSpace=BCS_BoneSpace
    translationMode=BMM_Additive
    translationSpace=BCS_BoneSpace
    alphaBoolBlend {blendInTime=0.2, blendOption=Linear, blendOutTime=0.4}
    boneToModify {boneName=dyn_shorts_bk_l}
    rotation=<rot -46.771 0 0>
    translation {x=12.414, z=-2.823}
   }
   n27 = "Transform (Modify) Bone - Bone: dyn_shorts_in_l"!(ComponentPose=@n26, bAlphaBoolEnabled:bool=True, Translation="0.000,0.363,1.892", Rotation="0.000,0.000,10.833", Scale="1,1,1") -> (Pose) {
    alphaInputType=Bool
    rotationMode=BMM_Additive
    rotationSpace=BCS_BoneSpace
    scaleMode=BMM_Additive
    scaleSpace=BCS_BoneSpace
    translationMode=BMM_Additive
    translationSpace=BCS_BoneSpace
    alphaBoolBlend {blendInTime=0.2, blendOption=Linear, blendOutTime=0.4}
    boneToModify {boneName=dyn_shorts_in_l}
    rotation {roll=10.833}
    translation {y=0.363, z=1.892}
   }
   n28 = "Transform (Modify) Bone - Bone: dyn_shorts_out_l"!(ComponentPose=@n27, bAlphaBoolEnabled:bool=True, Translation="-0.000,-0.000,4.679", Rotation="0.000,-17.429,-0.000", Scale="1,1,1") -> (Pose) {
    alphaInputType=Bool
    rotationMode=BMM_Additive
    rotationSpace=BCS_BoneSpace
    scaleMode=BMM_Additive
    scaleSpace=BCS_BoneSpace
    translationMode=BMM_Additive
    translationSpace=BCS_BoneSpace
    alphaBoolBlend {blendInTime=0.2, blendOption=Linear, blendOutTime=0.4}
    boneToModify {boneName=dyn_shorts_out_l}
    rotation {yaw=-17.429}
    translation {z=4.679}
   }
   n29 = Get bturnChar() -> (bturnChar:bool)
   n30 = "Transform (Modify) Bone - Bone: dyn_shorts_bk_r"(ComponentPose=@n28, bAlphaBoolEnabled=@n29, Translation="15.874,0.303,-3.559", Rotation="53.776,159.708,127.353", Scale="1,1.5,1") {
    alphaInputType=Bool
    rotationMode=BMM_Additive
    rotationSpace=BCS_BoneSpace
    scaleMode=BMM_Additive
    scaleSpace=BCS_BoneSpace
    translationMode=BMM_Additive
    translationSpace=BCS_BoneSpace
    alphaBoolBlend {blendInTime=0.2, blendOption=Linear, blendOutTime=0.35}
    alphaScaleBiasClamp {bInterpResult=true, interpSpeedDecreasing=5, interpSpeedIncreasing=8}
    boneToModify {boneName=dyn_shorts_bk_r}
    rotation=<rot 53.776 159.708 127.353>
    scale {y=1.5}
    translation=<15.874 0.303 -3.559>
   }
   n31 = "Transform (Modify) Bone - Bone: dyn_shorts_in_r"(ComponentPose=@n30, bAlphaBoolEnabled=@n29, Translation="0.133,-6.658,-3.704", Rotation="7.638,10.885,-66.089", Scale="1,1,1") {
    alphaInputType=Bool
    rotationMode=BMM_Additive
    rotationSpace=BCS_BoneSpace
    scaleMode=BMM_Additive
    scaleSpace=BCS_BoneSpace
    translationMode=BMM_Additive
    translationSpace=BCS_BoneSpace
    alphaBoolBlend {blendInTime=0.5, blendOption=Linear, blendOutTime=0.7}
    boneToModify {boneName=dyn_shorts_in_r}
    rotation=<rot 7.638 10.885 -66.089>
    translation=<0.133 -6.658 -3.704>
   }
   n32 = "Transform (Modify) Bone - Bone: dyn_shorts_out_r"(ComponentPose=@n31, bAlphaBoolEnabled=@n29, Translation="0.482,-3.259,1.022", Rotation="11.878,12.624,58.102", Scale="1,1,1") {
    alphaInputType=Bool
    rotationMode=BMM_Additive
    rotationSpace=BCS_BoneSpace
    scaleMode=BMM_Additive
    scaleSpace=BCS_BoneSpace
    translationMode=BMM_Additive
    translationSpace=BCS_BoneSpace
    alphaBoolBlend {blendInTime=0.5, blendOption=Linear, blendOutTime=0.7}
    boneToModify {boneName=dyn_shorts_out_r}
    rotation=<rot 11.878 12.624 58.102>
    translation=<0.482 -3.259 1.022>
   }
   n33 = "Transform (Modify) Bone - Bone: dyn_shorts_bk_l"(ComponentPose=@n32, bAlphaBoolEnabled=@n29, Translation="2.971,13.483,6.817", Rotation="6.900,20.284,38.638", Scale="1,1,1") {
    alphaInputType=Bool
    rotationMode=BMM_Additive
    rotationSpace=BCS_BoneSpace
    scaleMode=BMM_Additive
    scaleSpace=BCS_BoneSpace
    translationMode=BMM_Additive
    translationSpace=BCS_BoneSpace
    alphaBoolBlend {blendInTime=0.1, blendOption=Linear, blendOutTime=0.5}
    boneToModify {boneName=dyn_shorts_bk_l}
    rotation=<rot 6.9 20.284 38.638>
    translation=<2.971 13.483 6.817>
   }
   n34 = "Transform (Modify) Bone - Bone: dyn_shorts_in_l"(ComponentPose=@n33, bAlphaBoolEnabled=@n29, Translation="0.367,4.981,3.844", Rotation="-7.252,-6.154,-48.625", Scale="1,1,1") {
    alphaInputType=Bool
    rotationMode=BMM_Additive
    rotationSpace=BCS_BoneSpace
    scaleMode=BMM_Additive
    scaleSpace=BCS_BoneSpace
    translationMode=BMM_Additive
    translationSpace=BCS_BoneSpace
    alphaBoolBlend {blendInTime=0.5, blendOption=Linear, blendOutTime=0.7}
    boneToModify {boneName=dyn_shorts_in_l}
    rotation=<rot -7.252 -6.154 -48.625>
    translation=<0.367 4.981 3.844>
   }
   n35 = "Transform (Modify) Bone - Bone: dyn_shorts_out_l"(ComponentPose=@n34, bAlphaBoolEnabled=@n29, Translation="0.184,0.017,1.418", Rotation="1.074,-10.849,5.435", Scale="1,1,1") {
    alphaInputType=Bool
    rotationMode=BMM_Additive
    rotationSpace=BCS_BoneSpace
    scaleMode=BMM_Additive
    scaleSpace=BCS_BoneSpace
    translationMode=BMM_Additive
    translationSpace=BCS_BoneSpace
    alphaBoolBlend {blendInTime=0.5, blendOption=Linear, blendOutTime=0.7}
    boneToModify {boneName=dyn_shorts_out_l}
    rotation=<rot 1.074 -10.849 5.435>
    translation=<0.184 0.017 1.418>
   }
   n36 = "Transform (Modify) Bone - Bone: dyn_shorts_fr_r"(ComponentPose=@n35, bAlphaBoolEnabled=@n29, Translation="-0.650,-0.050,-0.020", Rotation="1.827,-9.326,8.629", Scale="1.000,1.472,1.000") {
    alphaInputType=Bool
    rotationMode=BMM_Additive
    rotationSpace=BCS_BoneSpace
    scaleMode=BMM_Additive
    scaleSpace=BCS_BoneSpace
    translationMode=BMM_Additive
    translationSpace=BCS_BoneSpace
    alphaBoolBlend {blendInTime=0.5, blendOption=Linear, blendOutTime=0.7}
    boneToModify {boneName=dyn_shorts_fr_r}
    rotation=<rot 1.827 -9.326 8.629>
    scale {y=1.472}
    translation=<-0.65 -0.05 -0.02>
   }
   n37 = "Transform (Modify) Bone - Bone: deform_glute_base_l"!(ComponentPose=@n36, bAlphaBoolEnabled:bool=@n29, Translation="0,0,-4", Rotation="0,-5,-10", Scale="1,1,1") -> (Pose) {
    alphaInputType=Bool
    rotationMode=BMM_Additive
    scaleMode=BMM_Additive
    translationMode=BMM_Additive
    alphaBoolBlend {blendInTime=0.5, blendOption=Linear, blendOutTime=0.7}
    boneToModify {boneName=deform_glute_base_l}
    rotation {roll=-10, yaw=-5}
    translation {z=-4}
   }
   n38 = "Transform (Modify) Bone - Bone: deform_glute_base_r"!(ComponentPose=@n37, bAlphaBoolEnabled:bool=@n29, Translation="0,0,-4", Rotation="5,0,-15", Scale="1,1,1") -> (Pose) {
    alphaInputType=Bool
    rotationMode=BMM_Additive
    scaleMode=BMM_Additive
    translationMode=BMM_Additive
    alphaBoolBlend {blendInTime=0.5, blendOption=Linear, blendOutTime=0.7}
    boneToModify {boneName=deform_glute_base_r}
    rotation {pitch=5, roll=-15}
    translation {z=-4}
   }
   n39 = "Transform (Modify) Bone - Bone: attach_backpack"(ComponentPose=@n38, Alpha=1, Translation="0,-6,0", Rotation="0,0,0", Scale="1,1,1") {
    translationMode=BMM_Additive
    translationSpace=BCS_BoneSpace
    alphaBoolBlend {blendOption=Linear}
    boneToModify {boneName=attach_backpack}
    translation {y=-6}
   }
   n40 = "Transform (Modify) Bone - Bone: dyn_shorts_bk_r"(ComponentPose=@n39, bAlphaBoolEnabled=True, Translation="0.699,-1.077,-2.505", Rotation="46.739,6.808,32.216", Scale="1.400,1.264,1.000") {
    alphaInputType=Bool
    rotationMode=BMM_Additive
    rotationSpace=BCS_BoneSpace
    scaleMode=BMM_Additive
    scaleSpace=BCS_BoneSpace
    translationMode=BMM_Additive
    translationSpace=BCS_BoneSpace
    alphaBoolBlend {blendInTime=0.5, blendOption=Linear, blendOutTime=0.3}
    boneToModify {boneName=dyn_shorts_bk_r}
    rotation=<rot 46.739 6.808 32.216>
    scale=<1.4 1.264>
    translation=<0.699 -1.077 -2.505>
   }
   n41 = "Transform (Modify) Bone - Bone: dyn_shorts_in_r"(ComponentPose=@n40, bAlphaBoolEnabled=True, Translation="0,0,0", Rotation="-15.713,-21.099,-12.734", Scale="1,1,1") {
    alphaInputType=Bool
    rotationMode=BMM_Additive
    rotationSpace=BCS_BoneSpace
    scaleMode=BMM_Additive
    scaleSpace=BCS_BoneSpace
    translationMode=BMM_Additive
    translationSpace=BCS_BoneSpace
    alphaBoolBlend {blendInTime=0.5, blendOption=Linear, blendOutTime=0.7}
    boneToModify {boneName=dyn_shorts_in_r}
    rotation=<rot -15.713 -21.099 -12.734>
   }
   n42 = "Transform (Modify) Bone - Bone: dyn_shorts_out_r"(ComponentPose=@n41, bAlphaBoolEnabled=True, Translation="0.168,-0.328,-1.243", Rotation="6.175,5.617,15.559", Scale="1,1,1") {
    alphaInputType=Bool
    rotationMode=BMM_Additive
    rotationSpace=BCS_BoneSpace
    scaleMode=BMM_Additive
    scaleSpace=BCS_BoneSpace
    translationMode=BMM_Additive
    translationSpace=BCS_BoneSpace
    alphaBoolBlend {blendInTime=0.5, blendOption=Linear, blendOutTime=0.7}
    boneToModify {boneName=dyn_shorts_out_r}
    rotation=<rot 6.175 5.617 15.559>
    translation=<0.168 -0.328 -1.243>
   }
   n43 = "Transform (Modify) Bone - Bone: dyn_shorts_bk_l"(ComponentPose=@n42, bAlphaBoolEnabled=True, Translation="-2,-1,8", Rotation="30,-20,0", Scale="1,1,1") {
    alphaInputType=Bool
    rotationMode=BMM_Additive
    rotationSpace=BCS_BoneSpace
    scaleMode=BMM_Additive
    scaleSpace=BCS_BoneSpace
    translationMode=BMM_Additive
    translationSpace=BCS_BoneSpace
    alphaBoolBlend {blendInTime=0.5, blendOption=Linear, blendOutTime=0.7}
    boneToModify {boneName=dyn_shorts_bk_l}
    rotation {pitch=30, yaw=-20}
    translation=<-2 -1 8>
   }
   n44 = "Transform (Modify) Bone - Bone: dyn_shorts_in_l"(ComponentPose=@n43, bAlphaBoolEnabled=True, Translation="0,0,0", Rotation="-0.134,16.622,-14.355", Scale="1,1,1") {
    alphaInputType=Bool
    rotationMode=BMM_Additive
    rotationSpace=BCS_BoneSpace
    scaleMode=BMM_Additive
    scaleSpace=BCS_BoneSpace
    translationMode=BMM_Additive
    translationSpace=BCS_BoneSpace
    alphaBoolBlend {blendInTime=0.5, blendOption=Linear, blendOutTime=0.7}
    boneToModify {boneName=dyn_shorts_in_l}
    rotation=<rot -0.134 16.622 -14.355>
   }
   n45 = "Transform (Modify) Bone - Bone: dyn_shorts_out_l"(ComponentPose=@n44, bAlphaBoolEnabled=True, Translation="0,0,0", Rotation="-0.152,-0.101,20", Scale="1,1,1") {
    alphaInputType=Bool
    rotationMode=BMM_Additive
    rotationSpace=BCS_BoneSpace
    scaleMode=BMM_Additive
    scaleSpace=BCS_BoneSpace
    translationMode=BMM_Additive
    translationSpace=BCS_BoneSpace
    alphaBoolBlend {blendInTime=0.5, blendOption=Linear, blendOutTime=0.7}
    boneToModify {boneName=dyn_shorts_out_l}
    rotation=<rot -0.152 -0.101 20>
   }
   n46 = "Transform (Modify) Bone - Bone: dyn_shorts_fr_r"(ComponentPose=@n45, bAlphaBoolEnabled=True, Translation="0,0,0", Rotation="0,0,0", Scale="1,1,1") {
    alphaInputType=Bool
    rotationMode=BMM_Additive
    rotationSpace=BCS_BoneSpace
    scaleMode=BMM_Additive
    scaleSpace=BCS_BoneSpace
    translationMode=BMM_Additive
    translationSpace=BCS_BoneSpace
    alphaBoolBlend {blendInTime=0.5, blendOption=Linear, blendOutTime=0.7}
    boneToModify {boneName=dyn_shorts_fr_r}
   }
   n47 = "Constraint - dyn_bag_01_MSTR (PPPPPPPP)"!(
    ComponentPose=@n46
    Alpha=1
    ConstraintWeights_0=0.345
    ConstraintWeights_1:float=0.0236
    ConstraintWeights_2:float=0.0364
    ConstraintWeights_3:float=0.0284
    ConstraintWeights_4:float=0.4247
    ConstraintWeights_5:float=0.0311
    ConstraintWeights_6:float=0.1107
    ConstraintWeights_7:float=0.0002
   ) -> (Pose) {
    alphaBoolBlend {blendOption=Linear}
    boneToModify {boneName=dyn_bag_01_MSTR}
    constraintSetup [
     {targetBone {boneName=pelvis}}
     {targetBone {boneName=spine_01}}
     {targetBone {boneName=spine_02}}
     {targetBone {boneName=spine_03}}
     {targetBone {boneName=thigh_twist_01_l}}
     {targetBone {boneName=deform_glute_base_l}}
     {targetBone {boneName=deform_groin_l}}
     {targetBone {boneName=thigh_twist_01_r}}
    ]
    constraintWeights [1,1,1,1,1,1,1,1]
   }
   n48 = "Constraint - dyn_bag_02_MSTR (PPPPPPP)"!(
    ComponentPose=@n47
    Alpha=1
    ConstraintWeights_0=0.4384
    ConstraintWeights_1:float=0.0274
    ConstraintWeights_2:float=0.0436
    ConstraintWeights_3:float=0.0353
    ConstraintWeights_4:float=0.2865
    ConstraintWeights_5:float=0.0757
    ConstraintWeights_6:float=0.0932
   ) -> (Pose) {
    alphaBoolBlend {blendOption=Linear}
    boneToModify {boneName=dyn_bag_02_MSTR}
    constraintSetup [
     {targetBone {boneName=pelvis}}
     {targetBone {boneName=spine_01}}
     {targetBone {boneName=spine_02}}
     {targetBone {boneName=spine_03}}
     {targetBone {boneName=thigh_twist_01_l}}
     {targetBone {boneName=deform_glute_base_l}}
     {targetBone {boneName=deform_groin_l}}
    ]
    constraintWeights [1,1,1,1,1,1,1]
   }
   n49 = "Constraint - dyn_bag_01_MSTR (R)"(ComponentPose=@n48, Alpha=1, ConstraintWeights_0=1) {
    alphaBoolBlend {blendOption=Linear}
    boneToModify {boneName=dyn_bag_01_MSTR}
    constraintSetup [{transformType=Rotation, targetBone {boneName=thigh_l}}]
    constraintWeights [1]
   }
   n50 = "Constraint - dyn_bag_02_MSTR (R)"(ComponentPose=@n49, Alpha=1, ConstraintWeights_0=1) {
    alphaBoolBlend {blendOption=Linear}
    boneToModify {boneName=dyn_bag_02_MSTR}
    constraintSetup [{transformType=Rotation, targetBone {boneName=thigh_l}}]
    constraintWeights [1]
   }
   n51 = "Constraint - dyn_bag_02_MSTR (PPPPPPP)"!(
    ComponentPose=@n50
    bAlphaBoolEnabled:bool=True
    ConstraintWeights_0=0.44
    ConstraintWeights_1=0.0219
    ConstraintWeights_2=0.0359
    ConstraintWeights_3=0.0305
    ConstraintWeights_4=0.2817
    ConstraintWeights_5=0.1217
    ConstraintWeights_6=0.0683
   ) -> (Pose) {
    alphaInputType=Bool
    alphaBoolBlend {blendOption=Linear}
    boneToModify {boneName=dyn_bag_02_MSTR}
    constraintSetup [
     {targetBone {boneName=pelvis}}
     {targetBone {boneName=spine_01}}
     {targetBone {boneName=spine_02}}
     {targetBone {boneName=spine_03}}
     {targetBone {boneName=thigh_twist_01_l}}
     {targetBone {boneName=deform_glute_base_l}}
     {targetBone {boneName=deform_groin_l}}
    ]
    constraintWeights [1,1,1,1,1,1,1]
   }
   n52 = "Transform (Modify) Bone - Bone: dyn_bag_01_MSTR"!(ComponentPose=@n51, bAlphaBoolEnabled:bool=True, Translation="0.830,1.631,-1.044", Rotation="9.859,-3.565,60.310", Scale="1,1,1") -> (Pose) {
    alphaInputType=Bool
    rotationMode=BMM_Additive
    rotationSpace=BCS_BoneSpace
    scaleMode=BMM_Additive
    scaleSpace=BCS_BoneSpace
    translationMode=BMM_Additive
    translationSpace=BCS_BoneSpace
    alphaBoolBlend {blendInTime=0.4, blendOption=Linear, blendOutTime=0.4}
    boneToModify {boneName=dyn_bag_01_MSTR}
    rotation=<rot 9.859 -3.565 60.31>
    translation=<0.83 1.631 -1.044>
   }
   n53 = "Transform (Modify) Bone - Bone: dyn_bag_02_MSTR"!(ComponentPose=@n52, bAlphaBoolEnabled:bool=True, Translation="-0.000,0.000,-2.049", Rotation="-2.718,0.822,-0.040", Scale="1,1,1") -> (Pose) {
    alphaInputType=Bool
    rotationMode=BMM_Additive
    rotationSpace=BCS_BoneSpace
    scaleMode=BMM_Additive
    scaleSpace=BCS_BoneSpace
    translationMode=BMM_Additive
    translationSpace=BCS_BoneSpace
    alphaBoolBlend {blendInTime=0.4, blendOption=Linear, blendOutTime=0.4}
    boneToModify {boneName=dyn_bag_02_MSTR}
    rotation=<rot -2.718 0.822 -0.04>
    translation {z=-2.049}
   }
   n54 = "Constraint - dyn_Map_MSTR (PPPPPPPP)"!(
    ComponentPose=@n53
    Alpha=1
    ConstraintWeights_0=0.4335
    ConstraintWeights_1:float=0.0235
    ConstraintWeights_2:float=0.037
    ConstraintWeights_3:float=0.0298
    ConstraintWeights_4:float=0.0002
    ConstraintWeights_5:float=0.2966
    ConstraintWeights_6:float=0.0719
    ConstraintWeights_7:float=0.1076
   ) -> (Pose) {
    alphaBoolBlend {blendOption=Linear}
    boneToModify {boneName=dyn_Map_MSTR}
    constraintSetup [
     {targetBone {boneName=pelvis}}
     {targetBone {boneName=spine_01}}
     {targetBone {boneName=spine_02}}
     {targetBone {boneName=spine_03}}
     {targetBone {boneName=thigh_twist_01_l}}
     {targetBone {boneName=thigh_twist_01_r}}
     {targetBone {boneName=deform_glute_base_r}}
     {targetBone {boneName=deform_groin_r}}
    ]
    constraintWeights [1,1,1,1,1,1,1,1]
   }
   n55 = "Constraint - dyn_Map_MSTR (R)"(ComponentPose=@n54, Alpha=0.7, ConstraintWeights_0=1) {
    alphaBoolBlend {blendOption=Linear}
    boneToModify {boneName=dyn_Map_MSTR}
    constraintSetup [{transformType=Rotation, targetBone {boneName=thigh_r}}]
    constraintWeights [1]
   }
   Output Pose(Result=Component To Local(ComponentPose=@n55)) {layerGroup=None}
   n56 = Control Rig(Source=Copy Pose From Mesh() {bUseAttachedParent=true}) {
    alphaBoolBlend {blendOption=Linear}
    controlRigAssetReference {blueprintRigClass=&/M_MED_BananaCake_CtrlRig.M_MED_BananaCake_CtrlRig_C}
   }
   n57 = RigidBody!(ComponentPose=Local To Component(LocalPose=@n56), Alpha, OverrideWorldGravity:Vector="0,0,0", ExternalForce) -> (Pose) {
    bClampLinearTranslationLimitToRefPose=true
    bOverrideWorldGravity=true
    overridePhysicsAsset=&/M_MED_BananaCake_RightShorts_Physics
    simulationSpace=BaseBoneSpace
    alphaBoolBlend {blendOption=Linear}
    baseBoneRef {boneName=thigh_twist_01_r}
    componentAppliedLinearAccClamp=<1000 1000 1000>
    componentLinearAccScale=<0.2 0.2 0.2>
    componentLinearVelScale=<0.2 0.2 0.2>
    overrideWorldGravity {z=-980}
    simSpaceSettings {
     maxAngularAcceleration=1000
     maxAngularVelocity=3
     maxLinearAcceleration=700
     maxLinearVelocity=500
     velocityScaleZ=0
     worldAlpha=1
     externalLinearDragV=<0.2 0.2>
    }
   }
   n58 = RigidBody(ComponentPose=@n57, OverrideWorldGravity="0,0,0") {
    bClampLinearTranslationLimitToRefPose=true
    bOverrideWorldGravity=true
    overridePhysicsAsset=&/M_MED_BananaCake_LeftShorts_Physics
    simulationSpace=BaseBoneSpace
    alphaBoolBlend {blendOption=Linear}
    baseBoneRef {boneName=thigh_twist_01_l}
    componentAppliedLinearAccClamp=<1000 1000 1000>
    componentLinearAccScale=<0.2 0.2 0.2>
    componentLinearVelScale=<0.2 0.2 0.2>
    overrideWorldGravity {z=-980}
    simSpaceSettings {
     maxAngularAcceleration=1000
     maxAngularVelocity=3
     maxLinearAcceleration=700
     maxLinearVelocity=500
     velocityScaleZ=0
     worldAlpha=1
     externalLinearDragV=<0.2 0.2>
    }
   }
   Save cached pose 'PreJacketSim'(Pose=Component To Local(ComponentPose=@n58)) {cacheName=PreJacketSim}
   Get Joint Location Vector() -> (Joint Location Vector:Vector)
   Get bnewVerCrouching() -> (bnewVerCrouching:bool)
  ]}
 EventGraph {nodes [
   n1 = Event OnCharacterCustomizationLoaded()
   > Set Backpack_part(Backpack_part:Skeletal Mesh Component Budgeted Object Reference=Get Array Item with EFort Custom Part Type(Array=@n1.CustomMeshParts)) -> (Output_Get:Skeletal Mesh Component Budgeted Object Reference)
   Event BlueprintUpdateAnimation()
   > n2 = Sequence!(execute) -> (
    then_0
    then_1
    then_2:Exec
    then_3:Exec
    then_4:Exec
    then_5:Exec
    then_6:Exec
    then_7:Exec
    then_8:Exec
    then_9:Exec
    then_10:Exec
    then_11:Exec
    then_12:Exec
    then_13:Exec
    then_14:Exec
   )
   > n3 = Gravity Override State Function DEPRECATED(CharacterPartAnimInstance=Self-Reference(), GameTimeInSeconds=GetGameTimeInSeconds(), GravityOverrideParams=Get Gravity Override Params())
   > Set GravityOverride(GravityOverride=@n3)
   n4 = GetOwnerSkeletalMeshComponent()
   n5 = BreakRotator(InRot=BreakTransform(InTransform=GetDeltaTransformFromRefPose(self=@n4, BoneName=head, BaseName=neck_02)).Rotation)
   n6 = Get X percentage() -> (X percentage:float)
   n7 = Get Y percentage() -> (Y percentage:float)
   n8 = Get Z percentage() -> (Z percentage:float)
   Set head rot(
    execute=@n2.then_1
    head rot:Rotator=MakeRotator(
     Roll=float * float!(A=@n5, B=-1, C:double=@n6) -> (ReturnValue)
     Pitch=float * float(A=@n5.Pitch, B=-1, C=@n7)
     Yaw=float * float(A=@n5.Yaw, B=-1, C=@n8)
    )
   ) -> (Output_Get:Rotator)
   n9 = GetDeltaTransformFromRefPose!(self=GetOwnerSkeletalMeshComponent(), BoneName=calf_r, BaseName=thigh_r) -> (ReturnValue_Location:Vector, ReturnValue_Rotation:Rotator, ReturnValue_Scale:Vector)
   n10 = MapRangeClamped(Value="Absolute (Float)"(A=BreakRotator(InRot=@n9.ReturnValue_Rotation).Yaw), InRangeA=60, InRangeB=100)
   > n11 = Branch(Condition=AND Boolean(A=Get bnewVerCrouching(), B="float > float"(A=@n10)))
   > Set Alpha Output(Alpha Output:float=@n10) -> (Output_Get:float)
   Set Alpha Output(execute=@n11.else, Alpha Output=0)
   n12 = Get bIsSkyDiving()
   n13 = NOT Boolean(A=Get bIsCrouching())
   n14 = Get bIsJumping()
   n15 = Get bIsFalling()
   n16 = Branch(execute=@n2.then_10, Condition=AND Boolean(A=OR Boolean!(A=@n14, B=@n15, C:bool=@n13, D:bool) -> (ReturnValue), B=NOT Boolean(A=@n12)))
   > Set CrouchFloat_01(CrouchFloat_01:double=1) -> (Output_Get:double)
   > Set CrouchFloat_02(CrouchFloat_02:double=0.5) -> (Output_Get:double)
   Set CrouchFloat_01(execute=@n16.else, CrouchFloat_01=0)
   > Set CrouchFloat_02(CrouchFloat_02=0)
   n17 = Get Backpack_part() -> (Backpack_part:Skeletal Mesh Component Budgeted Object Reference)
   n18 = GetSocketBoneName!(self:Skinned Mesh Component Object Reference=@n17, InSocketName:Name=Backpack_BR) -> (ReturnValue:Name=None)
   n19 = Get Joint Location Relative to Bone!(
    execute=@n2.then_11
    Target Joint:Name=@n18
    Relative Joint:Name=@n18
    Offset:Vector="0,-10,0"
    SkelMeshComponent:Mesh Component Object Reference=@n17
    Debug Draw:bool=false
   ) -> (then, Joint Location Vector:Vector) {is=title K2Node_CallFunction}
   > Set Joint Location Vector(Joint Location Vector:Vector=@n19) -> (Output_Get:Vector)
   n20 = "integer > integer"(A=Get CurrentLOD(), B=1)
   Set bCheckCurrentLod(execute=@n2.then_12, bCheckCurrentLod:bool=@n20) -> (Output_Get:bool)
   > Set bForBucketConstCond(bForBucketConstCond:bool=AND Boolean(A=@n20, B=Get bIsCrouching())) -> (Output_Get:bool)
   n21 = Cast To FortPlayerAnimInstance(execute=@n2.then_2, Object=GetAnimInstance(self=GetOwnerSkeletalMeshComponent()))
   > Set isAimingPistol(isAimingPistol:double=0) -> (Output_Get:double)
   > Set isPickaxe(isPickaxe:double=0) -> (Output_Get:double)
   > Set isAimingRifle(isAimingRifle:double=0) -> (Output_Get:double)
   > Set isPickaxeBoo(isPickaxeBoo:bool=false) -> (Output_Get:bool)
   > n22 = Switch on EFortWeaponCoreAnimation!(execute, Selection:EFortWeaponCoreAnimation Enum=Get WeaponCoreAnim(self=@n21)) -> (
    Melee:Exec
    Pistol:Exec
    Shotgun:Exec
    PaperBlueprint:Exec
    Rifle:Exec
    MeleeOneHand:Exec
    MachinePistol:Exec
    RocketLauncher:Exec
    GrenadeLauncher:Exec
    GoingCommando:Exec
    AssaultRifle:Exec
    TacticalShotgun:Exec
    SniperRifle:Exec
    TrapPlacement:Exec
    ShoulderLauncher:Exec
    AbilityDecoTool:Exec
    Crossbow:Exec
    C4:Exec
    RemoteControl:Exec
    DualWield:Exec
    AR_BullPup:Exec
    AR_ForwardGrip:Exec
    MedPackPaddles:Exec
    SMG_P90:Exec
    AR_DrumGun:Exec
    Consumable_Small:Exec
    Consumable_Large:Exec
    Balloon:Exec
    MountedTurret:Exec
    CreativeTool:Exec
    ExplosiveBow:Exec
    AshtonIndigo:Exec
    AshtonChicago:Exec
    MeleeDualWield:Exec
    Unarmed:Exec
   )
   n23 = Get bIsOnGround() -> (bIsOnGround:bool)
   Branch(
    execute=[
     @n22.AssaultRifle
     @n22.Crossbow
     @n22.GrenadeLauncher
     @n22.Rifle
     @n22.RocketLauncher
     @n22.Shotgun
     @n22.SniperRifle
     @n22.TacticalShotgun
    ]
    Condition=@n23
   )
   > Set isAimingRifle(isAimingRifle=1)
   Branch(execute=[@n22.MachinePistol, @n22.Pistol], Condition=@n23)
   > Set isAimingPistol(isAimingPistol=1)
   Branch(execute=@n22.Melee, Condition=@n23)
   > Set isPickaxe(isPickaxe=1)
   > Set isPickaxeBoo(isPickaxeBoo=true)
   n24 = Branch(execute=@n2.then_7, Condition=Get bIsOnGround())
   > Set OnGroundFloat(OnGroundFloat:double=1) -> (Output_Get:double)
   Set OnGroundFloat(execute=@n24.else, OnGroundFloat=0)
   n25 = Branch(execute=@n2.then_8, Condition=Get bIsSkydiveFloating())
   > Set SkydivingFloatFloat(SkydivingFloatFloat:double=1) -> (Output_Get:double)
   Set SkydivingFloatFloat(execute=@n25.else, SkydivingFloatFloat=0)
   n26 = Get bIsCrouching()
   n27 = Get bIsJumping()
   n28 = Get bIsFalling()
   n29 = Get bIsAccelerating2D()
   n30 = Branch(
    execute=@n2.then_9
    Condition=AND Boolean!(A=NOT Boolean(A=@n26), B=@n29, C:bool=NOT Boolean(A=@n27), D:bool=NOT Boolean(A=@n28)) -> (ReturnValue)
   )
   > Set Running(Running:double=1) -> (Output_Get:double)
   Set Running(execute=@n30.else, Running=0)
   n31 = Get bIsAccelerating2D()
   n32 = Branch(execute=@n2.then_5, Condition=AND Boolean!(A=Get bIsCrouching(), B=NOT Boolean(A=@n31)) -> (ReturnValue))
   > Set bnewVerCrouching(bnewVerCrouching:bool=true) -> (Output_Get:bool)
   > Set NewCrouchingFloat(NewCrouchingFloat:double=1) -> (Output_Get:double)
   Set bnewVerCrouching(execute=@n32.else, bnewVerCrouching=false)
   > Set NewCrouchingFloat(NewCrouchingFloat=0)
   n33 = Sequence!(execute=@n2.then_13) -> (then_0, then_1, then_2)
   > Delay(Duration=0.3)
   n34 = "Not Equal (Float)"(A="Get CurrentValue Z (Yaw)_0"() -> ("CurrentValue Z (Yaw)_0":float), B=Get NewVar() -> (NewVar:double))
   > Branch(Condition=@n34)
   > Set NewVar(NewVar:double="Get CurrentValue Z (Yaw)_0"()) -> (Output_Get:double)
   n35 = NOT Boolean(A=Get bIsAccelerating2D())
   n36 = NOT Boolean(A=@n34)
   Set bturnChar(execute=@n33.then_1, bturnChar:bool=AND Boolean(A=AND Boolean(A=@n36, B=@n35), B=Get bIsCrouching())) -> (Output_Get:bool)
   > "Set CurrentValue Z (Yaw)_0"("CurrentValue Z (Yaw)_0":float=GetActorForwardVector!(self:Actor Object Reference=TryGetFortPawn()) -> (ReturnValue_X:double, ReturnValue_Y:double, ReturnValue_Z:double)) -> (Output_Get:float)
   > Set NewVar_1(NewVar_1:bool=AND Boolean(A=@n34, B=Get bIsCrouching())) -> (Output_Get:bool)
   n37 = Branch(execute=@n2.then_6, Condition=OR Boolean!(A=Get bIsSkyDiving(), B=Get bIsParachuteOpened()) -> (ReturnValue))
   > Set SkydivingFloat(SkydivingFloat:double=1) -> (Output_Get:double)
   Set SkydivingFloat(execute=@n37.else, SkydivingFloat=0)
   n38 = Get bIsAccelerating2D()
   n39 = Get bIsCrouching()
   n40 = Branch(execute=@n2.then_4, Condition=OR Boolean(A=AND Boolean(A=@n39, B=@n38), B=Get NewVar_1() -> (NewVar_1:bool)))
   > Set CrouchingAndMoving(CrouchingAndMoving:bool=true) -> (Output_Get:bool)
   > Set CrouchingAndMovingFloat(CrouchingAndMovingFloat:double=1) -> (Output_Get:double)
   Set CrouchingAndMoving(execute=@n40.else, CrouchingAndMoving=false)
   > Set CrouchingAndMovingFloat(CrouchingAndMovingFloat=0)
   n41 = Get bIsJumping()
   n42 = Get bIsAccelerating2D()
   n43 = Branch(
    execute=@n2.then_3
    Condition=OR Boolean!(A=OR Boolean(A=@n41, B=false), B=AND Boolean!(A=false, B=false, C:bool=@n42) -> (ReturnValue), C:bool) -> (ReturnValue)
   )
   > Set NewJumping(NewJumping:bool=true) -> (Output_Get:bool)
   Set NewJumping(execute=@n43.else, NewJumping=false)
   OR Boolean(A=NOT Boolean(A=Get bIsSkyDiving()), B=false)
   "Byte > Byte"(A:Byte, B:bool=@n29) -> (ReturnValue:bool=~) {is=title K2Node_PromotableOperator}
   Get bIsFalling()
   Get bIsFalling()
  ]}
]
variables [
 Alpha Output {type=number}
 Backpack_part {type=SkeletalMeshComponentBudgeted}
 bCheckCurrentLod {type=boolean}
 bForBucketConstCond {type=boolean}
 bnewVerCrouching {type=boolean}
 bturnChar {type=boolean}
 CrouchFloat_01 {type=number}
 CrouchFloat_02 {type=number}
 CrouchingAndMoving {type=boolean}
 CrouchingAndMovingFloat {type=number}
 "CurrentValue Z (Yaw)_0" {type=number}
 Gravity Override Params {type=GravityOverrideParamsStruct}
 GravityOverride {type=Vector}
 head rot {type=Rotator}
 isAimingPistol {type=number}
 isAimingRifle {type=number}
 isPickaxe {type=number}
 isPickaxeBoo {type=boolean}
 Joint Location Vector {type=Vector}
 NewCrouchingFloat {type=number}
 NewJumping {type=boolean}
 NewVar {type=number}
 NewVar_1 {type=boolean}
 OnGroundFloat {type=number}
 Running {type=number}
 SkydivingFloat {type=number}
 SkydivingFloatFloat {type=number}
 X percentage {type=number}
 Y percentage {type=number}
 Z percentage {type=number}
]
